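# Physics asset for the RockerPunkCube1H one-handed pickaxe skeletal mesh, a Battle Royale melee weapon cosmetic.
# It approximates the weapon with a single capsule body on the b_AR_Root bone, offset and rotated from that bone.
# It defines no constraints, so there are no jointed or separately simulated parts.
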PhysicsAsset /BRCosmetics/Weapons/FORT_Melee/Pickaxe_RockerPunkCube1H/Meshes/SK_Pickaxe_RockerPunkCube1H_PhysicsAsset
bodies {b_AR_Root [{shapeType=Capsule, length=31.6409, radius=27.603, center=<2.5822 0 11.3816>, rotation=<rot 0 -90 -2.9524>}]}
constraints []
